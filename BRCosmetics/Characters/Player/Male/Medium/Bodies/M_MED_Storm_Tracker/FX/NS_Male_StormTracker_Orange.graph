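# Orange-variant Niagara effect for the male Storm Tracker character body cosmetic, made up of Glow_Background and Light emitters.
# It exposes a User.bisHandHidden bool, and its dependencies include bool-to-float and bool-to-Vector2D dynamic inputs, likely driven by that flag.
# System scripts use SystemState plus a Set Variables assignment at spawn.
# No renderers or data interfaces are listed for either emitter.
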NiagaraSystem /BRCosmetics/Characters/Player/Male/Medium/Bodies/M_MED_Storm_Tracker/FX/NS_Male_StormTracker_Orange
dependencies {
 usedDataInterfaces []
 usedDynamicInputs [&MakeCustomFloatFromBool, &MakeCustomVector2DFromBool]
 usedModules [
  &SystemState
  &self:SystemSpawnScript.SystemScriptSource.SystemScriptGraph.NiagaraNodeAssignment_0.SetVariables_#1
 ]
}
emitters {Glow_Background {rendererClasses [], renderers []}, Light {rendererClasses [], renderers []}}
summary {
 systemName=NS_Male_StormTracker_Orange
 emitters [{emitterName=Glow_Background, rendererClasses []}, {emitterName=Light, rendererClasses []}]
 userVariables [User.bisHandHidden {type=NiagaraBool}]
}
