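# Physics asset for the Storm Tracker male medium body cosmetic, covering the secondary motion of its jacket and shirt bones.
# Box bodies on three back jacket bones and the right front one are constrained to spine_04, and the right shirt bone to the pelvis.
# All joints have tight angular limits, with the center back jacket swinging widest on one axis.
# Calf capsules carry no constraints.
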PhysicsAsset /BRCosmetics/Characters/Player/Male/Medium/Bodies/M_MED_Storm_Tracker/Meshes/M_MED_Storm_Tracker_Physics
bodies {
 calf_l [{shapeType=Capsule, length=26.0663, radius=10.365, center=<30.8261 -0.24 2.1182>, rotation=<rot 89.6205 180 -170.6978>}]
 calf_r [{shapeType=Capsule, length=21.5206, radius=9.9221, center=<-24.276 2.6671 -1.2393>, rotation=<rot 88.7014 0.0001 6.6653>}]
 dyn_jacket_back [{shapeType=Box, extentX=3.4415, extentY=8.6463, extentZ=11.8093, center=<3.3926 0.5071 0.0408>, rotation=<rot -2.5289 -104.1173 -1.1193>}]
 dyn_jacket_back_l [{shapeType=Box, extentX=10.3884, extentY=1.8109, extentZ=10.1436, center=<2.6941 1.3522 0.5902>, rotation=<rot -1.0451 -164.0344 -78.4769>}]
 dyn_jacket_back_r [{shapeType=Box, extentX=11.6901, extentY=1.9095, extentZ=10.9474, center=<3.71 -0.135 -0.1634>, rotation=<rot -15.6628 104.5484 -89.3596>}]
 dyn_jacket_front_r [{shapeType=Box, extentX=11.7288, extentY=1.01, extentZ=8.624, center=<2.2397 1.5719 -0.8632>, rotation=<rot -4.885 -178.5105 -74.0049>}]
 dyn_shirt_r_master [{shapeType=Box, extentX=8.3347, extentY=1.1153, extentZ=11.3397, center=<10.8656 -0.962 0.3976>, rotation=<rot 0 5.3829 -80.4249>}]
 pelvis [{shapeType=Box, extentX=3.4843, extentY=3.4843, extentZ=3.4843}]
 spine_04 [{shapeType=Box, extentX=2.3795, extentY=3.3477, extentZ=2.6518, center=<0.7813 0.8786 -0.2676>, rotation=<rot -0.1141 -90 -3.2317>}]
}
constraints [
 {bone1Name=dyn_jacket_back, bone2Name=spine_04, swing1LimitDegrees=35, swing2LimitDegrees=5, twistLimitDegrees=5}
 {bone1Name=dyn_jacket_back_l, bone2Name=spine_04, swing1LimitDegrees=10, swing2LimitDegrees=15, twistLimitDegrees=5}
 {bone1Name=dyn_jacket_back_r, bone2Name=spine_04, swing1LimitDegrees=10, swing2LimitDegrees=15, twistLimitDegrees=5}
 {bone1Name=dyn_jacket_front_r, bone2Name=spine_04, swing1LimitDegrees=10, swing2LimitDegrees=10, twistLimitDegrees=5}
 {bone1Name=dyn_shirt_r_master, bone2Name=pelvis, swing1LimitDegrees=5, swing2LimitDegrees=10, twistLimitDegrees=2}
]
