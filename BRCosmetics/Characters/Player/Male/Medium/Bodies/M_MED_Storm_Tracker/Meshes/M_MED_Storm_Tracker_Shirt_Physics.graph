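# Physics asset for the shirt mesh of the Storm Tracker male medium player cosmetic, covering a dynamic shirt chain on the left side.
# Three thin box bodies on the dyn_shirt_l bones are constrained in series, with the first attached to the left thigh twist bone.
# Each joint has matching swing limits and a tight twist limit.
# A capsule on the left thigh and a box on its twist bone supply the leg bodies.
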PhysicsAsset /BRCosmetics/Characters/Player/Male/Medium/Bodies/M_MED_Storm_Tracker/Meshes/M_MED_Storm_Tracker_Shirt_Physics
bodies {
 dyn_shirt_l_01 [{shapeType=Box, extentX=1.25, extentY=6.5, extentZ=6.5, center=<1 0 0>}]
 dyn_shirt_l_02 [{shapeType=Box, extentX=1.25, extentY=6.5, extentZ=6.5, center=<1 0 0>}]
 dyn_shirt_l_03 [{shapeType=Box, extentX=1.25, extentY=6.5, extentZ=6.5, center=<0.4 0 0>}]
 thigh_l [{shapeType=Capsule, length=29.4416, radius=7.5506, center=<-22.6543 0.9069 0.7864>, rotation=<rot -0.1691 85.9191 -92.1248>}]
 thigh_twist_01_l [{shapeType=Box, extentX=7.5, extentY=7.5, extentZ=7.5}]
}
constraints [
 {bone1Name=dyn_shirt_l_01, bone2Name=thigh_twist_01_l, swing1LimitDegrees=45, swing2LimitDegrees=45, twistLimitDegrees=5}
 {bone1Name=dyn_shirt_l_02, bone2Name=dyn_shirt_l_01, swing1LimitDegrees=45, swing2LimitDegrees=45, twistLimitDegrees=5}
 {bone1Name=dyn_shirt_l_03, bone2Name=dyn_shirt_l_02, swing1LimitDegrees=45, swing2LimitDegrees=45, twistLimitDegrees=5}
]
