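# Physics asset for the Storm Tracker male medium body's jacket drawstrings, apparently for simulating their dangle and collision against the torso.
# Left and right strings are each a four-bone chain of small capsules, linked segment to segment by tight swing and twist limits.
# A spine_05 body with three large capsules stands in for the upper torso as a collider.
# No constraint joins either chain root to the spine.
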PhysicsAsset /BRCosmetics/Characters/Player/Male/Medium/Bodies/M_MED_Storm_Tracker/Meshes/M_MED_Storm_Tracker_Strings_Physics
bodies {
 dyn_jacketstring_l_01 [{shapeType=Capsule, length=1.4865, radius=1.1472, rotation=<rot 2.0565 -77.317 80.938>}]
 dyn_jacketstring_l_02 [{shapeType=Capsule, length=3.7665, radius=1.2422, center=<2.42 0.1133 0.206>, rotation=<rot 0 92.6803 -85.1435>}]
 dyn_jacketstring_l_03 [{shapeType=Capsule, length=4.067, radius=1.1755, center=<2.5769 0.1054 -0.0419>, rotation=<rot 0 -87.6578 -89.0694>}]
 dyn_jacketstring_l_04 [{shapeType=Capsule, length=2.3656, radius=1.4044, center=<1.7144 -0.0395 -0.1459>, rotation=<rot 17.2766 87.1654 -95.0963>}]
 dyn_jacketstring_r_01 [{shapeType=Capsule, length=1.1036, radius=1.09, rotation=<rot 0 -89.9996 -87.9564>}]
 dyn_jacketstring_r_02 [{shapeType=Capsule, length=4.7407, radius=1.1229, center=<3.2828 -0.1626 0.5548>, rotation=<rot 0 89.2293 -83.0491>}]
 dyn_jacketstring_r_03 [{shapeType=Capsule, length=3.7513, radius=1.0205, center=<3.3336 -0.0667 -0.0856>, rotation=<rot 0 -92.0898 -86.4617>}]
 dyn_jacketstring_r_04 [{shapeType=Capsule, length=1.0239, radius=1.3379, center=<1.1469 0.231 -0.0658>, rotation=<rot -0.3387 -87.2117 -93.5768>}]
 spine_05 [
  {shapeType=Capsule, length=4.151, radius=19.8711, center=<-9.603 -6.8972 -0.069>, rotation=<rot 89.4165 -179.9936 101.8651>}
  {shapeType=Capsule, length=15, radius=15, center=<-17.3584 0.1041 -0.0001>, rotation=<rot 90 0 4.5091>}
  {shapeType=Capsule, length=21.542, radius=6.6662, center=<-10.7166 8.4818 0.0002>}
 ]
}
constraints [
 {bone1Name=dyn_jacketstring_l_03, bone2Name=dyn_jacketstring_l_02, swing1LimitDegrees=20, swing2LimitDegrees=5, twistLimitDegrees=5}
 {bone1Name=dyn_jacketstring_l_04, bone2Name=dyn_jacketstring_l_03, swing1LimitDegrees=10, swing2LimitDegrees=5, twistLimitDegrees=5}
 {bone1Name=dyn_jacketstring_l_02, bone2Name=dyn_jacketstring_l_01, swing1LimitDegrees=30, swing2LimitDegrees=5, twistLimitDegrees=5}
 {bone1Name=dyn_jacketstring_r_02, bone2Name=dyn_jacketstring_r_01, swing1LimitDegrees=20, swing2LimitDegrees=5, twistLimitDegrees=5}
 {bone1Name=dyn_jacketstring_r_03, bone2Name=dyn_jacketstring_r_02, swing1LimitDegrees=20, swing2LimitDegrees=5, twistLimitDegrees=5}
 {bone1Name=dyn_jacketstring_r_04, bone2Name=dyn_jacketstring_r_03, swing1LimitDegrees=20, swing2LimitDegrees=5, twistLimitDegrees=5}
]
