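# Post-process style cosmetic AnimBlueprint for the M_MED_Storm_Tracker body, adding secondary motion to shirt, jacket, strings and headphones.
# It copies the attached parent's pose, blends headphone constraints between neck and spine by head yaw, then chains three RigidBody simulations.
# Additive bone offsets follow, some weighted by shirt_alpha, which rises during skydive and parachute states.
# AnimDynamics and leg-swing branches are disconnected from output; jacket_ext_force is set around jumps but unused.
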AnimBlueprint /BRCosmetics/Characters/Player/Male/Medium/Bodies/M_MED_Storm_Tracker/Meshes/M_MED_Storm_Tracker_AnimBp
target_skeleton=&/M_MED_Storm_Tracker_Skeleton
graphs [
 AnimGraph {nodes [
   n1 = Anim Dynamics - Bone: dyn_shirt_l_master(Alpha=0.8, ExternalForce="0,0,0") {
    angularDampingOverride=1
    angularSpringConstant=50
    bAngularSpring=true
    bGravityOverrideInSimSpace=true
    bOverrideLinearDamping=true
    bUseGravityOverride=true
    bUseSphericalLimits=true
    linearDampingOverride=1
    simulationSpace=RootRelative
    alphaBoolBlend {blendOption=Linear}
    boundBone {boneName=dyn_shirt_l_master}
    componentAppliedLinearAccClamp=<800 800 800>
    componentLinearAccScale=<0.25 0.25 0.5>
    componentLinearVelScale=<0.5 0.5 0.5>
    gravityOverride {z=-600}
    physicsBodyDefinitions [{
      sphereCollisionRadius=0
      boundBone {boneName=dyn_shirt_l_master}
      boxExtents=<10 10 3>
      constraintSetup {angularLimitsMax=<0 40 20>, angularLimitsMin=<0 -10 -20>, angularTarget=<1 0 0>}
      localJointOffset=<6 0 0>
     }]
    relativeSpaceBone {boneName=pelvis}
    sphericalLimits [{limitRadius=10, drivingBone {boneName=thigh_l}, sphereLocalOffset=<-10 1 0>}]
   }
   Anim Dynamics - Bone: dyn_shirt_l_03(ComponentPose=@n1, Alpha=0.8, ExternalForce="0,0,0") {
    angularDampingOverride=1
    angularSpringConstant=25
    bAngularSpring=true
    bGravityOverrideInSimSpace=true
    bOverrideAngularDamping=true
    bOverrideLinearDamping=true
    bUseGravityOverride=true
    bUseSphericalLimits=true
    linearDampingOverride=1
    simulationSpace=RootRelative
    alphaBoolBlend {blendOption=Linear}
    boundBone {boneName=dyn_shirt_l_03}
    componentAppliedLinearAccClamp=<4000 4000 4000>
    componentLinearAccScale=<0.5 0.5 0.5>
    componentLinearVelScale=<0.5 0.5 0.5>
    gravityOverride {z=-600}
    physicsBodyDefinitions [{
      sphereCollisionRadius=0
      boundBone {boneName=dyn_shirt_l_03}
      boxExtents=<10 10 3>
      constraintSetup {angularLimitsMax=<20 60 60>, angularLimitsMin=<-30 -30 -50>, angularTarget=<1 0 0>}
      localJointOffset=<6 0 0>
     }]
    relativeSpaceBone {boneName=pelvis}
    sphericalLimits [{limitRadius=10, drivingBone {boneName=thigh_twist_01_l}, sphereLocalOffset=<10 0 2.5>}]
   }
   n2 = Anim Dynamics - Bone: dyn_shirt_l_01(Alpha=1, ExternalForce="0,0,0") {
    angularDampingOverride=0.98
    bGravityOverrideInSimSpace=true
    bUseGravityOverride=true
    numSolverIterationsPostUpdate=2
    numSolverIterationsPreUpdate=8
    alphaBoolBlend {blendOption=Linear}
    boundBone {boneName=dyn_shirt_l_01}
    componentLinearAccScale=<1 1 1>
    componentLinearVelScale=<1 1 1>
    gravityOverride {z=-300}
    physicsBodyDefinitions [{
      sphereCollisionRadius=0
      boundBone {boneName=dyn_shirt_l_01}
      boxExtents=<5 20 20>
      constraintSetup {angularTargetAxis=AxisY, twistAxis=AxisZ, angularLimitsMax=<0 30 30>, angularLimitsMin=<0 -30 -30>, angularTarget=<0 0 0>}
      localJointOffset=<0 0 -3>
     }]
    relativeSpaceBone {boneName=root}
   }
   n3 = Anim Dynamics - Bone: dyn_shirt_l_02(ComponentPose=@n2, Alpha=1, ExternalForce="0,0,0") {
    angularDampingOverride=0.98
    bGravityOverrideInSimSpace=true
    bUseGravityOverride=true
    numSolverIterationsPostUpdate=2
    numSolverIterationsPreUpdate=8
    alphaBoolBlend {blendOption=Linear}
    boundBone {boneName=dyn_shirt_l_02}
    componentLinearAccScale=<1 1 1>
    componentLinearVelScale=<1 1 1>
    gravityOverride {z=-300}
    physicsBodyDefinitions [{
      sphereCollisionRadius=0
      boundBone {boneName=dyn_shirt_l_02}
      boxExtents=<5 20 20>
      constraintSetup {angularTargetAxis=AxisY, twistAxis=AxisZ, angularLimitsMax=<0 30 30>, angularLimitsMin=<0 -30 -30>, angularTarget=<0 0 0>}
      localJointOffset=<0 0 -1>
     }]
    relativeSpaceBone {boneName=root}
   }
   "Transform (Modify) Bone - Bone: dyn_shirt_l_master"(ComponentPose=@n3, Alpha=1, Translation="0,1,1.2", Rotation="0,0,0", Scale="1,1,1") {
    rotationMode=BMM_Additive
    rotationSpace=BCS_BoneSpace
    translationMode=BMM_Additive
    translationSpace=BCS_BoneSpace
    alphaBoolBlend {blendOption=Linear}
    boneToModify {boneName=dyn_shirt_l_master}
    translation {y=1, z=1.2}
   }
   n4 = Get legSwingDelta() -> (legSwingDelta:double)
   n5 = "Transform (Modify) Bone - Bone: dyn_shirt_r_master"(Alpha=@n4, Translation="0,2,2", Rotation="0,0,0", Scale="1,1,1") {
    rotationMode=BMM_Additive
    rotationSpace=BCS_BoneSpace
    alphaBoolBlend {blendOption=Linear}
    boneToModify {boneName=dyn_shirt_r_master}
    translation {y=2, z=2}
   }
   n6 = "Transform (Modify) Bone - Bone: dyn_shirt_l_master"(ComponentPose=@n5, Alpha=@n4, Translation="0,0,0", Rotation="0,0,0", Scale="1,1,1") {
    rotationMode=BMM_Additive
    rotationSpace=BCS_BoneSpace
    translationMode=BMM_Additive
    translationSpace=BCS_BoneSpace
    alphaBoolBlend {blendOption=Linear}
    boneToModify {boneName=dyn_shirt_l_master}
   }
   "Transform (Modify) Bone - Bone: dyn_shirt_l_02"(ComponentPose=@n6, Alpha=1, Translation="0,0,0", Rotation="0,0,0", Scale="1,1,1") {
    rotationMode=BMM_Additive
    rotationSpace=BCS_BoneSpace
    translationMode=BMM_Additive
    translationSpace=BCS_BoneSpace
    alphaBoolBlend {blendOption=Linear}
    boneToModify {boneName=dyn_shirt_l_02}
   }
   n7 = "Constraint - dyn_headphones (PP)"!(
    ComponentPose=Local To Component(LocalPose=Copy Pose From Mesh() {bUseAttachedParent=true})
    Alpha=Get headSwingDeltaINV() -> (headSwingDeltaINV:double)
    ConstraintWeights_0=0
    ConstraintWeights_1:float=1
   ) -> (Pose) {
    alphaBoolBlend {blendOption=Linear}
    boneToModify {boneName=dyn_headphones}
    constraintSetup [{targetBone {boneName=neck_02}}, {targetBone {boneName=spine_05}}]
    constraintWeights [1,1]
   }
   n8 = "Constraint - dyn_headphones (PP)"(ComponentPose=@n7, Alpha=Get headSwingDelta() -> (headSwingDelta:double), ConstraintWeights_0=1, ConstraintWeights_1=0) {
    alphaBoolBlend {blendOption=Linear}
    alphaScaleBiasClamp {bMapRange=true}
    boneToModify {boneName=dyn_headphones}
    constraintSetup [{targetBone {boneName=neck_02}}, {targetBone {boneName=spine_05}}]
    constraintWeights [1,1]
   }
   n9 = "Constraint - dyn_shirt_l_master (P)"(ComponentPose=@n8, Alpha=1, ConstraintWeights_0=1) {
    alphaBoolBlend {blendOption=Linear}
    boneToModify {boneName=dyn_shirt_l_master}
    constraintSetup [{targetBone {boneName=pelvis}}]
    constraintWeights [1]
   }
   n10 = RigidBody(ComponentPose=@n9) {
    bOverrideWorldGravity=true
    overridePhysicsAsset=&/M_MED_Storm_Tracker_Shirt_Physics
    alphaBoolBlend {blendOption=Linear}
    baseBoneRef {boneName=root}
    overrideWorldGravity {z=-980}
   }
   n11 = RigidBody(ComponentPose=@n10) {
    bClampLinearTranslationLimitToRefPose=true
    bOverrideWorldGravity=true
    overridePhysicsAsset=&/M_MED_Storm_Tracker_Strings_Physics
    simulationSpace=BaseBoneSpace
    alphaBoolBlend {blendOption=Linear}
    baseBoneRef {boneName=pelvis}
    componentAppliedLinearAccClamp=<7000 7000 7000>
    componentLinearAccScale=<0.5 0.5 0.5>
    componentLinearVelScale=<0.5 0.5 0.5>
    overrideWorldGravity {z=-980}
   }
   n12 = RigidBody(ComponentPose=@n11) {
    bClampLinearTranslationLimitToRefPose=true
    bOverrideWorldGravity=true
    overridePhysicsAsset=&/M_MED_Storm_Tracker_Physics
    simulationSpace=BaseBoneSpace
    alphaBoolBlend {blendOption=Linear}
    baseBoneRef {boneName=pelvis}
    componentAppliedLinearAccClamp=<4000 4000 4000>
    componentLinearAccScale=<0.25 0.25 0.25>
    componentLinearVelScale=<0.5 0.5 0.5>
    overrideWorldGravity {z=-980}
   }
   n13 = Get shirt_alpha() -> (shirt_alpha:double)
   n14 = "Transform (Modify) Bone - Bone: dyn_shirt_l_master"(ComponentPose=@n12, Alpha=@n13, Translation="9,8,8", Rotation="30,30,10", Scale="1,1,1") {
    rotationMode=BMM_Additive
    rotationSpace=BCS_BoneSpace
    alphaBoolBlend {blendOption=Linear}
    boneToModify {boneName=dyn_shirt_l_master}
    rotation=<rot 30 30 10>
    translation=<9 8 8>
   }
   n15 = "Transform (Modify) Bone - Bone: dyn_shirt_l_02"(ComponentPose=@n14, Alpha=@n13, Translation="0,3,0", Rotation="2,0,-0.25", Scale="1,1,1") {
    rotationMode=BMM_Additive
    rotationSpace=BCS_BoneSpace
    translationMode=BMM_Additive
    translationSpace=BCS_BoneSpace
    alphaBoolBlend {blendOption=Linear}
    boneToModify {boneName=dyn_shirt_l_02}
    rotation {pitch=2, roll=-0.25}
    translation {y=3}
   }
   n16 = "Transform (Modify) Bone - Bone: dyn_shirt_l_03"(ComponentPose=@n15, Alpha=@n13, Translation="0,4,10", Rotation="45,0,0", Scale="1,1,1") {
    rotationMode=BMM_Additive
    rotationSpace=BCS_BoneSpace
    translationMode=BMM_Additive
    translationSpace=BCS_BoneSpace
    alphaBoolBlend {blendOption=Linear}
    boneToModify {boneName=dyn_shirt_l_03}
    rotation {pitch=45}
    translation {y=4, z=10}
   }
   n17 = "Transform (Modify) Bone - Bone: dyn_jacket_back"(ComponentPose=@n16, Alpha=Get shirt_alpha(), Translation="0,0,0", Rotation="1,0,0", Scale="1,1,1") {
    rotationMode=BMM_Additive
    rotationSpace=BCS_BoneSpace
    translationMode=BMM_Additive
    translationSpace=BCS_BoneSpace
    alphaBoolBlend {blendOption=Linear}
    boneToModify {boneName=dyn_jacket_back}
    rotation {pitch=1}
   }
   n18 = "Transform (Modify) Bone - Bone: dyn_jacket_back_l"(ComponentPose=@n17, Alpha=1, Translation="0,0,0", Rotation="0.3,0,0", Scale="1,1,1") {
    rotationMode=BMM_Additive
    rotationSpace=BCS_BoneSpace
    translationMode=BMM_Additive
    translationSpace=BCS_BoneSpace
    alphaBoolBlend {blendOption=Linear}
    boneToModify {boneName=dyn_jacket_back_l}
    rotation {pitch=0.3}
   }
   n19 = "Transform (Modify) Bone - Bone: dyn_jacket_back"(ComponentPose=@n18, Alpha=1, Translation="0,-3,0", Rotation="1,0,0", Scale="1,1,1") {
    rotationMode=BMM_Additive
    rotationSpace=BCS_BoneSpace
    translationMode=BMM_Additive
    translationSpace=BCS_BoneSpace
    alphaBoolBlend {blendOption=Linear}
    boneToModify {boneName=dyn_jacket_back}
    rotation {pitch=1}
    translation {y=-3}
   }
   Output Pose(Result=Component To Local(ComponentPose=@n19)) {layerGroup=None}
   Get jacket_ext_force() -> (jacket_ext_force:Vector)
  ]}
 EventGraph {nodes [
   Event BlueprintUpdateAnimation()
   n1 = Get bIsSkydiveDiveMode()
   n2 = Get bIsSkydiveFloating()
   n3 = Get bIsSkyDiving()
   n4 = Get bIsParachuteOpened()
   > Set shirt_alpha(shirt_alpha:double=SelectFloat(A=0.25, B=0, bPickA=OR Boolean!(A=@n3, B=@n4, C:bool=@n2, D:bool=@n1) -> (ReturnValue))) -> (Output_Get:double)
   n5 = BreakTransform(InTransform=GetDeltaTransformFromRefPose(self=GetOwnerSkeletalMeshComponent(), BoneName=head, BaseName=spine_05))
   n6 = MapRangeClamped(Value=BreakRotator(InRot=@n5.Rotation).Yaw, InRangeA=25, InRangeB=-43)
   > Set headSwingDelta(headSwingDelta:double=@n6) -> (Output_Get:double)
   n7 = BreakRotator(InRot=@n5.Rotation)
   > Set headTwist_r(headTwist_r:double=MapRangeClamped(Value=@n7, InRangeA=14, InRangeB=25)) -> (Output_Get:double)
   > Set headTwist_l(headTwist_l:double=MapRangeClamped(Value=@n7, InRangeA=-9, InRangeB=-20)) -> (Output_Get:double)
   > Set headSwingDeltaINV(headSwingDeltaINV:double=float - float(B=@n6) {is=title K2Node_CallFunction}) -> (Output_Get:double)
   n8 = Branch(Condition=Get bIsJumping())
   > Delay(Duration=0.25)
   > Set jacket_ext_force(jacket_ext_force:Vector="0,0,1250") -> (Output_Get:Vector)
   > n9 = Delay(Duration=0.375)
   Do Once(execute=@n8.else, Reset=@n9.then)
   > Set jacket_ext_force(jacket_ext_force="0, 0, 0")
   n10 = BreakTransform(InTransform=GetDeltaTransformFromRefPose(BoneName=pelvis, BaseName=thigh_twist_01_l))
   Set legSwingDelta(legSwingDelta:double=MapRangeClamped(Value=BreakRotator(InRot=@n10.Rotation).Yaw, InRangeA=24, InRangeB=-8)) -> (Output_Get:double)
   n11 = BreakTransform(InTransform=GetDeltaTransformFromRefPose(BoneName=pelvis, BaseName=thigh_l))
   > Set legSwingDelta(legSwingDelta=MapRangeClamped(Value=BreakRotator(InRot=@n11.Rotation).Yaw, InRangeA=24, InRangeB=-8))
  ]}
]
variables [
 headSwingDelta {type=number}
 headSwingDeltaINV {type=number}
 headTwist_l {type=number}
 headTwist_r {type=number}
 jacket_ext_force {type=Vector}
 legSwingDelta {type=number}
 shirt_alpha {type=number}
]
